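# Control Rig for the MorningSoak CookCrutch male medium body, driving its custom dyn_ bones from the standard player skeleton.
# Custom pelvis and spine follow standard bones, then spine_01 is constrained to the custom spine's top.
# Left arm bones use B-spline IK through controls blended from arm bones and neighboring controls, with finger bones constrained separately.
# Both custom legs derive rotations from standard legs, then FABRIK pins them to offset ball transforms.
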ControlRigBlueprint /BRCosmetics/Characters/Player/Male/Medium/Bodies/M_MED_MorningSoak_CookCrutch/Meshes/M_MED_MorningSoak_CookCrutch_CtrlRig
graphs [
 RigVMFunctionLibrary {links [], nodes []}
 RigVMFunction_Sequence_1_ContainedGraph {nodes [
   n1 = RigVMFunctionEntryNode()
   RigVMFunction_Sequence = &RigVMFunction_Sequence(ExecuteContext=@n1.ExecuteContext)
   n2 = &RigVMFunction_Sequence(ExecuteContext=@RigVMFunction_Sequence.B)
   n3 = &RigVMFunction_Sequence(ExecuteContext=@n2.B)
   RigVMFunctionReturnNode(A=@RigVMFunction_Sequence.A, B=@n2.A, C=@n3.A, D=@n3.B)
  ]}
 RigVMModel {nodes [
   n1 = &RigUnit_PrepareForExecution()
   Make_Array = RigVMDispatchNode(Values="((Type=Bone,Name=\"dyn_L_splineArm_clavicle\"),(Type=Bone,Name=\"dyn_L_splineArm_1\"),((),Type=Bone,Name=\"dyn_L_splineArm_2\"),((),Type=Bone,Name=\"dyn_L_splineArm_3\"),((),Type=Bone,Name=\"dyn_L_splineArm_4\"),((),Type=Bone,Name=\"dyn_L_splineArm_5\"),((),Type=Bone, ...(truncated)") {template="DISPATCH_RigVMDispatch_ArrayMake(in Values,out Array)"}
   RigVMVariableNode(Value=@Make_Array.Array, ExecuteContext=@n1.ExecutePin) {variable=L arm custom bones}
   Make_Array_1 = RigVMDispatchNode(Values="((Type=Control,Name=\"dyn_L_splineArm_clavicle_ctrl\"),((),Type=Control,Name=\"dyn_L_splineArm_1_ctrl\"),((),Type=Control,Name=\"dyn_L_splineArm_2_ctrl\"),((),Type=Control,Name=\"dyn_L_splineArm_3_ctrl\"),((),Type=Control,Name=\"dyn_L_splineArm_4_ctrl\"),((),Type=Co ...(truncated)") {template="DISPATCH_RigVMDispatch_ArrayMake(in Values,out Array)"}
   > RigVMVariableNode(Value=@Make_Array_1.Array) {variable=L arm controls}
   > RigVMVariableNode(Value=/Script/ControlRig.RigUnit_HierarchyGetPoseItemArray(ElementType=All, Initial=False, ItemsToGet=@Make_Array_1.Array) {is=struct RigVMUnitNode}.Pose) {variable=L arm bind pose}
   RigUnit_BeginExecution = &RigUnit_BeginExecution()
   RigVMFunction_Sequence = RigVMAggregateNode(ExecuteContext=@RigUnit_BeginExecution.ExecutePin)
   &RigUnit_ParentConstraint(
    AdvancedSettings=(InterpolationType=Average,RotationOrderForFilter=XZY)
    Child=(Type=Bone,Name="dyn_custom_pelvis")
    Filter=(TranslationFilter=(bX=true,bY=true,bZ=true),RotationFilter=(bX=True,bY=True,bZ=True),ScaleFilter=(bX=True,bY=True,bZ=True))
    Parents=((Item=(Type=Bone,Name="pelvis"),Weight=1))
    Weight=1
    ExecutePin=@RigVMFunction_Sequence.A
   )
   > &RigUnit_ParentConstraint(
    AdvancedSettings=(InterpolationType=Average,RotationOrderForFilter=XZY)
    Child=(Type=Bone,Name="dyn_custom_spine_01")
    Filter=(TranslationFilter=(bX=true,bY=true,bZ=true),RotationFilter=(bX=True,bY=True,bZ=True),ScaleFilter=(bX=True,bY=True,bZ=True))
    Parents=((Item=(Type=Bone,Name="spine_01"),Weight=1))
    Weight=1
   )
   > &RigUnit_ParentConstraint(
    AdvancedSettings=(InterpolationType=Average,RotationOrderForFilter=XZY)
    Child=(Type=Bone,Name="dyn_custom_spine_02")
    Filter=(TranslationFilter=(bX=true,bY=true,bZ=true),RotationFilter=(bX=True,bY=True,bZ=True),ScaleFilter=(bX=True,bY=True,bZ=True))
    Parents=((Item=(Type=Bone,Name="spine_02"),Weight=1))
    Weight=1
   )
   > &RigUnit_ParentConstraint(
    AdvancedSettings=(InterpolationType=Average,RotationOrderForFilter=XZY)
    Child=(Type=Bone,Name="dyn_custom_spine_03")
    Filter=(TranslationFilter=(bX=true,bY=true,bZ=true),RotationFilter=(bX=True,bY=True,bZ=True),ScaleFilter=(bX=True,bY=True,bZ=True))
    Parents=((Item=(Type=Bone,Name="spine_03"),Weight=1))
    Weight=1
   )
   > &RigUnit_ParentConstraint(
    AdvancedSettings=(InterpolationType=Average,RotationOrderForFilter=XZY)
    Child=(Type=Bone,Name="spine_01")
    Parents=((Item=(Type=Bone,Name="dyn_custom_spine_03"),Weight=1))
    Weight=1
   )
   &RigUnit_ParentConstraint(
    AdvancedSettings=(InterpolationType=Average,RotationOrderForFilter=XZY)
    Child=(Type=Null,Name="dyn_L_splineArm_hand_ctrl null")
    Parents=((Item=(Type=Bone,Name="hand_l"),Weight=1))
    Weight=1
    bMaintainOffset=true
    ExecutePin=@RigVMFunction_Sequence.B
   )
   > &RigUnit_ParentConstraint(
    AdvancedSettings=(InterpolationType=Average,RotationOrderForFilter=XZY)
    Child=(Type=Null,Name="dyn_L_splineArm_clavicle_ctrl null")
    Parents=((Item=(Type=Bone,Name="clavicle_l"),Weight=0.5),(Item=(Type=Bone,Name="spine_05"),Weight=0.5))
    Weight=1
   )
   > &RigUnit_ParentConstraint(
    AdvancedSettings=(InterpolationType=Average,RotationOrderForFilter=XZY)
    Child=(Type=Null,Name="dyn_L_splineArm_1_ctrl null")
    Parents=((Item=(Type=Bone,Name="upperarm_l"),Weight=1))
    Weight=1
   )
   > &RigUnit_ParentConstraint(
    AdvancedSettings=(InterpolationType=Average,RotationOrderForFilter=XZY)
    Child=(Type=Null,Name="dyn_L_splineArm_4_ctrl null")
    Parents=((Item=(Type=Bone,Name="lowerarm_l"),Weight=0.5),(Item=(Type=Bone,Name="upperarm_l"),Weight=0.5))
    Weight=1
   )
   > &RigUnit_ParentConstraint(
    AdvancedSettings=(InterpolationType=Average,RotationOrderForFilter=XZY)
    Child=(Type=Null,Name="dyn_L_splineArm_3_ctrl null")
    Parents=((Item=(Type=Control,Name="dyn_L_splineArm_2_ctrl"),Weight=0.5),(Item=(Type=Control,Name="dyn_L_splineArm_4_ctrl"),Weight=0.5))
    Weight=1
   )
   > &RigUnit_ParentConstraint(
    AdvancedSettings=(InterpolationType=Average,RotationOrderForFilter=XZY)
    Child=(Type=Null,Name="dyn_L_splineArm_2_ctrl null")
    Parents=((Item=(Type=Control,Name="dyn_L_splineArm_1_ctrl"),Weight=0.5),(Item=(Type=Control,Name="dyn_L_splineArm_3_ctrl"),Weight=0.5))
    Weight=1
   )
   > &RigUnit_ParentConstraint(
    AdvancedSettings=(InterpolationType=Average,RotationOrderForFilter=XZY)
    Child=(Type=Null,Name="dyn_L_splineArm_5_ctrl null")
    Parents=((Item=(Type=Control,Name="dyn_L_splineArm_4_ctrl"),Weight=0.75),(Item=(Type=Control,Name="dyn_L_splineArm_7_ctrl"),Weight=0.25))
    Weight=1
    bMaintainOffset=true
   )
   > &RigUnit_ParentConstraint(
    AdvancedSettings=(InterpolationType=Average,RotationOrderForFilter=XZY)
    Child=(Type=Null,Name="dyn_L_splineArm_7_ctrl null")
    Parents=((Item=(Type=Bone,Name="lowerarm_l"),Weight=0.75),(Item=(Type=Control,Name="dyn_L_splineArm_hand_ctrl"),Weight=0.25))
    Weight=1
   )
   > &RigUnit_ParentConstraint(
    AdvancedSettings=(InterpolationType=Average,RotationOrderForFilter=XZY)
    Child=(Type=Null,Name="dyn_L_splineArm_6_ctrl null")
    Parents=((Item=(Type=Control,Name="dyn_L_splineArm_5_ctrl"),Weight=0.5),(Item=(Type=Control,Name="dyn_L_splineArm_7_ctrl"),Weight=0.5))
    Weight=1
   )
   > SplineIK = RigVMFunctionReferenceNode(
    Primary Axis=(X=1,Y=0,Z=0)
    Samples Per Segment=16
    Secondary Spline Direction=(X=0,Y=1,Z=0)
    Spline Mode=BSpline
    Stretch=true
    Up Axis=(X=0,Y=1,Z=0)
    debug=false
    Controls=RigVMVariableNode() {variable=L arm controls}.Value
    Bones=RigVMVariableNode() {variable=L arm custom bones}.Value
   )
   n2 = &RigUnit_GetTransform(Item=(Type=Bone,Name="dyn_L_custom_thigh"), bInitial=true)
   n3 = &RigUnit_GetTransform(Item=(Type=Bone,Name="thigh_l"), bInitial=true)
   Set Transform_3 = &RigUnit_SetRotation(
    Item=(Type=Bone,Name="dyn_L_custom_thigh")
    Weight=1
    bInitial=False
    Value=&RigVMFunction_MathTransformMakeRelative(
     Global=&RigUnit_GetTransform(Item=(Type=Bone,Name="thigh_l"), Space=LocalSpace, bInitial=false).Transform
     Parent=&RigVMFunction_MathTransformMakeRelative(Parent=@n2.Transform, Global=@n3.Transform).Local
    ).Local.Rotation
    ExecutePin=@RigVMFunction_Sequence.C
   )
   n4 = &RigUnit_GetTransform(Item=(Type=Bone,Name="dyn_L_custom_calf"), bInitial=true)
   n5 = &RigUnit_GetTransform(Item=(Type=Bone,Name="calf_l"), bInitial=true)
   > Set Transform = &RigUnit_SetRotation(
    Item=(Type=Bone,Name="dyn_L_custom_calf")
    Weight=1
    bInitial=False
    Value=&RigVMFunction_MathTransformMakeRelative(
     Global=&RigUnit_GetTransform(Item=(Type=Bone,Name="calf_l"), Space=LocalSpace, bInitial=false).Transform
     Parent=&RigVMFunction_MathTransformMakeRelative(Parent=@n4.Transform, Global=@n5.Transform).Local
    ).Local.Rotation
   )
   n6 = &RigUnit_GetTransform(Item=(Type=Bone,Name="foot_l"), bInitial=true)
   n7 = &RigUnit_GetTransform(Item=(Type=Bone,Name="dyn_L_custom_foot"), bInitial=true)
   > Set Transform_1 = &RigUnit_SetRotation(
    Item=(Type=Bone,Name="dyn_L_custom_foot")
    Weight=1
    bInitial=false
    Value=&RigVMFunction_MathTransformMakeRelative(
     Global=&RigUnit_GetTransform(Item=(Type=Bone,Name="foot_l"), Space=LocalSpace, bInitial=false).Transform
     Parent=&RigVMFunction_MathTransformMakeRelative(Global=@n6.Transform, Parent=@n7.Transform).Local
    ).Local.Rotation
   )
   Multiply = &RigVMFunction_MathTransformMul(
    B=&RigUnit_GetTransform(Item=(Type=Bone,Name="ball_l"), bInitial=false).Transform
    A=&RigVMFunction_MathTransformMakeRelative(
     Parent=&RigUnit_GetTransform(Item=(Type=Bone,Name="ball_l"), bInitial=true).Transform
     Global=&RigUnit_GetTransform(Item=(Type=Bone,Name="dyn_L_custom_ball"), bInitial=true).Transform
    ).Local
   ) {template="Multiply::Execute(in A,in B,out Result)"}
   > /Script/ControlRig.RigUnit_FABRIKItemArray(
    Items=((Type=Bone,Name="dyn_L_custom_thigh"),(Type=Bone,Name="dyn_L_custom_ball"))
    MaxIterations=10
    Precision=1
    Weight=1
    bPropagateToChildren=True
    bSetEffectorTransform=True
    EffectorTransform=@Multiply.Result
   ) {is=struct RigVMUnitNode}
   > &RigUnit_SetTransform(Item=(Type=Bone,Name="dyn_L_custom_ball"), Weight=1, bInitial=False, Value=@Multiply.Result)
   n8 = &RigUnit_GetTransform(Item=(Type=Bone,Name="dyn_R_custom_thigh"), bInitial=true)
   n9 = &RigUnit_GetTransform(Item=(Type=Bone,Name="thigh_r"), bInitial=true)
   > Set Transform_7 = &RigUnit_SetRotation(
    Item=(Type=Bone,Name="dyn_R_custom_thigh")
    Weight=1
    bInitial=False
    Value=&RigVMFunction_MathTransformMakeRelative(
     Global=&RigUnit_GetTransform(Item=(Type=Bone,Name="thigh_r"), Space=LocalSpace, bInitial=false).Transform
     Parent=&RigVMFunction_MathTransformMakeRelative(Parent=@n8.Transform, Global=@n9.Transform).Local
    ).Local.Rotation
   )
   n10 = &RigUnit_GetTransform(Item=(Type=Bone,Name="dyn_R_custom_calf"), bInitial=true)
   n11 = &RigUnit_GetTransform(Item=(Type=Bone,Name="calf_r"), bInitial=true)
   > Set Transform_4 = &RigUnit_SetRotation(
    Item=(Type=Bone,Name="dyn_R_custom_calf")
    Weight=1
    bInitial=False
    Value=&RigVMFunction_MathTransformMakeRelative(
     Global=&RigUnit_GetTransform(Item=(Type=Bone,Name="calf_r"), Space=LocalSpace, bInitial=false).Transform
     Parent=&RigVMFunction_MathTransformMakeRelative(Parent=@n10.Transform, Global=@n11.Transform).Local
    ).Local.Rotation
   )
   n12 = &RigUnit_GetTransform(Item=(Type=Bone,Name="foot_r"), bInitial=true)
   n13 = &RigUnit_GetTransform(Item=(Type=Bone,Name="dyn_R_custom_foot"), bInitial=true)
   > Set Transform_5 = &RigUnit_SetRotation(
    Item=(Type=Bone,Name="dyn_R_custom_foot")
    Weight=1
    bInitial=false
    Value=&RigVMFunction_MathTransformMakeRelative(
     Global=&RigUnit_GetTransform(Item=(Type=Bone,Name="foot_r"), Space=LocalSpace, bInitial=false).Transform
     Parent=&RigVMFunction_MathTransformMakeRelative(Global=@n12.Transform, Parent=@n13.Transform).Local
    ).Local.Rotation
   )
   n14 = &RigVMFunction_MathTransformMul(
    B=&RigUnit_GetTransform(Item=(Type=Bone,Name="ball_r"), bInitial=false).Transform
    A=&RigVMFunction_MathTransformMakeRelative(
     Parent=&RigUnit_GetTransform(Item=(Type=Bone,Name="ball_r"), bInitial=true).Transform
     Global=&RigUnit_GetTransform(Item=(Type=Bone,Name="dyn_R_custom_ball"), bInitial=true).Transform
    ).Local
   ) {template="Multiply::Execute(in A,in B,out Result)"}
   > &RigUnit_FABRIKItemArray(
    Items=((Type=Bone,Name="dyn_R_custom_thigh"),(Type=Bone,Name="dyn_R_custom_ball"))
    MaxIterations=10
    Precision=1
    Weight=1
    bPropagateToChildren=True
    bSetEffectorTransform=True
    EffectorTransform=@n14.Result
   )
   > &RigUnit_SetTransform(Item=(Type=Bone,Name="dyn_R_custom_ball"), Weight=1, bInitial=False, Value=@n14.Result)
   &RigUnit_ParentConstraint(
    AdvancedSettings=(InterpolationType=Average,RotationOrderForFilter=XZY)
    Child=(Type=Bone,Name="dyn_L_splineArm_metacarpal")
    Parents=((Item=(Type=Bone,Name="middle_metacarpal_l"),Weight=1))
    Weight=1
    ExecutePin=@RigVMFunction_Sequence.D
   )
   > &RigUnit_ParentConstraint(
    AdvancedSettings=(InterpolationType=Average,RotationOrderForFilter=XZY)
    Child=(Type=Bone,Name="dyn_L_splineArm_finger")
    Parents=((Item=(Type=Bone,Name="middle_03_l"),Weight=1))
    Weight=1
   )
   > &RigUnit_ParentConstraint(
    AdvancedSettings=(InterpolationType=Average,RotationOrderForFilter=XZY)
    Child=(Type=Bone,Name="dyn_L_custom_thumb_01")
    Parents=((Item=(Type=Bone,Name="thumb_01_l"),Weight=1))
    Weight=1
   )
   > &RigUnit_ParentConstraint(
    AdvancedSettings=(InterpolationType=Average,RotationOrderForFilter=XZY)
    Child=(Type=Bone,Name="dyn_L_custom_thumb_02")
    Parents=((Item=(Type=Bone,Name="thumb_02_l"),Weight=1))
    Weight=1
   )
   > &RigUnit_ParentConstraint(
    AdvancedSettings=(InterpolationType=Average,RotationOrderForFilter=XZY)
    Child=(Type=Bone,Name="dyn_L_custom_thumb_03")
    Parents=((Item=(Type=Bone,Name="thumb_03_l"),Weight=1))
    Weight=1
   )
   &RigUnit_ModifyTransforms(
    ExecutePin=()
    ItemToModify="((Item=(Type=Bone,Name=\"lowerarm_l\"),Transform=(Rotation=(X=0,Y=0,Z=-0.4617,W=-0.887),Translation=(X=0,Y=0,Z=0),Scale3D=(X=1,Y=1,Z=1))),(Item=(Type=Bone,Name=\"upperarm_l\"),Transform=(Rotation=(X=0,Y=0,Z=0.2164,W=0.9763),Translation=(X=0,Y=0,Z=0),Scale3D=(X ...(truncated)"
    Weight=1
    WeightMinimum=0
   )
  ]}
]
hierarchy [
 root {loc=<0 0 0>, rot=<0 0 0 1>}
 attach {parent=root, loc=<0 0 0>, rot=<0 0 0 1>}
 pelvis {parent=root, loc=<0 -0.7071 93.1191>, rot=<0.0127 0.707 -0.0127 -0.707>}
 spine_01 {parent=pelvis, loc=<6.7494 -0.6388 0>, rot=<0 0 0.034 -0.9994>}
 spine_02 {parent=spine_01, loc=<7.2518 -0.035 0>, rot=<0 0 0.0238 -0.9997>}
 spine_03 {parent=spine_02, loc=<6.0104 0 0>, rot=<0 0 0.0149 -0.9999>}
 spine_04 {parent=spine_03, loc=<8.1024 0 0>, rot=<0 0 -0.0132 -0.9999>}
 spine_05 {parent=spine_04, loc=<10.0146 0 0>, rot=<0 0 -0.0015 -1>}
 clavicle_l {parent=spine_05, loc=<4.3575 -0.1466 -2.3223>, rot=<0.0465 -0.7289 -0.009 -0.6829>}
 upperarm_l {parent=clavicle_l, loc=<15.5111 0 0>, rot=<0.0226 -0.3234 -0.0088 -0.9459>}
 lowerarm_l {parent=upperarm_l, loc=<26.6514 0 0>, rot=<0.0391 -0.0099 -0.2453 -0.9686>}
 hand_l {parent=lowerarm_l, loc=<22.3491 0 0>, rot=<-0.7081 -0.0507 0.0975 -0.6975>}
 index_metacarpal_l {parent=hand_l, loc=<1.6535 0.0183 -2.96>, rot=<0.0158 -0.0124 -0.0173 -0.9996>}
 index_01_l {parent=index_metacarpal_l, loc=<8.3008 0.0357 -0.0524>, rot=<-0.0044 0.0485 0.0924 -0.9945>}
 index_02_l {parent=index_01_l, loc=<3.3587 -0.3004 0.0452>, rot=<-0.002 0.0027 0.3225 -0.9466>}
 index_03_l {parent=index_02_l, loc=<2.5578 -0.123 -0.0252>, rot=<-0.002 0.0067 0.2558 -0.9667>}
 middle_metacarpal_l {parent=hand_l, loc=<2.1856 -0.1725 -0.5984>, rot=<0.006 0.0168 -0.0329 -0.9993>}
 middle_01_l {parent=middle_metacarpal_l, loc=<7.3275 -0.2447 0.0129>, rot=<-0.0208 0.0042 0.0858 -0.9961>}
 middle_02_l {parent=middle_01_l, loc=<4.0123 0.1401 0.1175>, rot=<-0.0218 0.0258 0.3512 -0.9357>}
 middle_03_l {parent=middle_02_l, loc=<2.6077 0.0534 0.1373>, rot=<0.0082 0.0188 0.1708 -0.9851>}
 pinky_metacarpal_l {parent=hand_l, loc=<1.7616 -0.9753 3.3814>, rot=<-0.0397 0.0814 -0.069 -0.9935>}
 pinky_01_l {parent=pinky_metacarpal_l, loc=<7.1709 -0.4831 -0.1094>, rot=<-0.0654 -0.0035 0.2242 -0.9723>}
 pinky_02_l {parent=pinky_01_l, loc=<2.8274 0.0893 -0.1043>, rot=<0.0037 0.0011 0.2854 -0.9584>}
 pinky_03_l {parent=pinky_02_l, loc=<2.2046 0.0307 -0.0078>, rot=<-0.0002 0.0098 0.1445 -0.9895>}
 ring_metacarpal_l {parent=hand_l, loc=<2.2074 -0.3625 1.3638>, rot=<-0.0284 0.0722 -0.0458 -0.9959>}
 ring_01_l {parent=ring_metacarpal_l, loc=<7.0295 -0.6197 -0.1446>, rot=<-0.0489 -0.0252 0.1892 -0.9804>}
 ring_02_l {parent=ring_01_l, loc=<3.6174 0.4647 -0.1027>, rot=<0 0 0.2806 -0.9598>}
 ring_03_l {parent=ring_02_l, loc=<2.4152 0.114 0.0199>, rot=<0 0 0.1189 -0.9929>}
 thumb_01_l {parent=hand_l, loc=<1.3118 -0.3191 -2.459>, rot=<0.6546 -0.4042 0.0664 -0.6354>}
 thumb_02_l {parent=thumb_01_l, loc=<5.9336 -0.0448 0.0401>, rot=<-0.0037 0.0213 0.298 -0.9543>}
 thumb_03_l {parent=thumb_02_l, loc=<2.2223 0.3682 -0.0907>, rot=<0.0118 -0.013 0.0066 -0.9998>}
 weapon_l {parent=hand_l, loc=<-0.1974 -3.4111 0.3811>, rot=<0 0 0 1>}
 lowerarm_twist_01_l {parent=lowerarm_l, loc=<17.8549 0 0>, rot=<0 0 0 1>}
 lowerarm_twist_02_l {parent=lowerarm_l, loc=<9.6199 0 0>, rot=<0 0 0 1>}
 upperarm_twist_01_l {parent=upperarm_l, loc=<0.0041 0 0>, rot=<0 0 0 1>}
 upperarm_twist_02_l {parent=upperarm_l, loc=<13.7941 0 0>, rot=<0 0 0 1>}
 clavicle_r {parent=spine_05, loc=<4.3575 -0.1466 2.3223>, rot=<0.7289 0.0465 -0.6829 0.009>}
 upperarm_r {parent=clavicle_r, loc=<-15.5111 0 0>, rot=<0.0226 -0.3234 -0.0088 -0.9459>}
 lowerarm_r {parent=upperarm_r, loc=<-26.6515 0 0>, rot=<0.0391 -0.0099 -0.2453 -0.9686>}
 hand_r {parent=lowerarm_r, loc=<-22.3492 0 0>, rot=<-0.7081 -0.0507 0.0975 -0.6975>}
 index_metacarpal_r {parent=hand_r, loc=<-1.6534 -0.0181 2.9601>, rot=<0.0158 -0.0124 -0.0173 -0.9996>}
 index_01_r {parent=index_metacarpal_r, loc=<-8.3008 -0.0357 0.0524>, rot=<-0.0044 0.0485 0.0924 -0.9945>}
 index_02_r {parent=index_01_r, loc=<-3.3589 0.3003 -0.0453>, rot=<-0.002 0.0027 0.3225 -0.9466>}
 index_03_r {parent=index_02_r, loc=<-2.5578 0.1231 0.0251>, rot=<-0.002 0.0067 0.2558 -0.9667>}
 middle_metacarpal_r {parent=hand_r, loc=<-2.1856 0.1726 0.5984>, rot=<0.006 0.0168 -0.0329 -0.9993>}
 middle_01_r {parent=middle_metacarpal_r, loc=<-7.3273 0.2447 -0.0129>, rot=<-0.0208 0.0042 0.0858 -0.9961>}
 middle_02_r {parent=middle_01_r, loc=<-4.0142 -0.1411 -0.07>, rot=<-0.0217 0.0259 0.3512 -0.9357>}
 middle_03_r {parent=middle_02_r, loc=<-2.6074 -0.0534 -0.1312>, rot=<0.0083 0.0188 0.1708 -0.9851>}
 pinky_metacarpal_r {parent=hand_r, loc=<-1.7617 0.9753 -3.3814>, rot=<-0.0397 0.0814 -0.069 -0.9935>}
 pinky_01_r {parent=pinky_metacarpal_r, loc=<-7.1707 0.4832 0.1094>, rot=<-0.0654 -0.0035 0.2242 -0.9723>}
 pinky_02_r {parent=pinky_01_r, loc=<-2.8274 -0.0892 0.1043>, rot=<0.0037 0.0011 0.2854 -0.9584>}
 pinky_03_r {parent=pinky_02_r, loc=<-2.2045 -0.031 0.0078>, rot=<-0.0002 0.0098 0.1445 -0.9895>}
 ring_metacarpal_r {parent=hand_r, loc=<-2.2073 0.3627 -1.3638>, rot=<-0.0284 0.0722 -0.0458 -0.9959>}
 ring_01_r {parent=ring_metacarpal_r, loc=<-7.0297 0.6197 0.1446>, rot=<-0.0489 -0.0252 0.1892 -0.9804>}
 ring_02_r {parent=ring_01_r, loc=<-3.6171 -0.4648 0.1027>, rot=<0 0 0.2806 -0.9598>}
 ring_03_r {parent=ring_02_r, loc=<-2.4153 -0.1141 -0.0199>, rot=<0 0 0.1189 -0.9929>}
 thumb_01_r {parent=hand_r, loc=<-1.3119 0.3192 2.459>, rot=<0.6546 -0.4042 0.0664 -0.6354>}
 thumb_02_r {parent=thumb_01_r, loc=<-5.9335 0.0447 -0.0401>, rot=<-0.0037 0.0213 0.298 -0.9543>}
 thumb_03_r {parent=thumb_02_r, loc=<-2.2223 -0.3682 0.0907>, rot=<0.0118 -0.013 0.0066 -0.9998>}
 weapon_r {parent=hand_r, loc=<0.1975 3.4112 -0.3811>, rot=<0 0 0 -1>}
 lowerarm_twist_01_r {parent=lowerarm_r, loc=<-17.855 0 0>, rot=<0 0 0 1>}
 lowerarm_twist_02_r {parent=lowerarm_r, loc=<-9.6199 0 0>, rot=<0 0 0 1>}
 upperarm_twist_01_r {parent=upperarm_r, loc=<-0.0041 0 0>, rot=<0 0 0 1>}
 upperarm_twist_02_r {parent=upperarm_r, loc=<-13.7942 0 0>, rot=<0 0 0 1>}
 neck_01 {parent=spine_05, loc=<11.3138 0.1529 0>, rot=<0 0 -0.2003 -0.9797>}
 neck_02 {parent=neck_01, loc=<4.5627 0 0>, rot=<0 0 0.0504 -0.9987>}
 head {parent=neck_02, loc=<5.3914 0 0>, rot=<0 0 0.115 -0.9934>}
 dyn_L_cartman_arm {parent=spine_05, loc=<12.7658 1.1213 -10.3133>, rot=<-0.309 -0.9144 -0.057 -0.2553>}
 dyn_L_cartman_hand {parent=dyn_L_cartman_arm, loc=<11.8371 0 0>, rot=<0.2367 -0.205 0.0196 -0.9495>}
 dyn_R_cartman_arm {parent=spine_05, loc=<12.7662 1.1214 10.3133>, rot=<0.9144 -0.309 -0.2553 0.057>}
 dyn_R_cartman_hand {parent=dyn_R_cartman_arm, loc=<-11.8373 0.0001 0.0001>, rot=<0.2367 -0.205 0.0196 -0.9495>}
 dyn_cartman_head {parent=spine_05, loc=<10.411 1.726 0>, rot=<0.0284 -0.7065 -0.0284 -0.7065>}
 dyn_L_splineArm_clavicle {parent=spine_05, loc=<1.3173 -0.8055 -10.5135>, rot=<0.0465 -0.7289 -0.009 -0.6829>}
 dyn_L_splineArm_1 {parent=dyn_L_splineArm_clavicle, loc=<10.9662 0.0022 -1.3166>, rot=<0.0151 -0.2535 -0.0215 -0.967>}
 dyn_L_splineArm_2 {parent=dyn_L_splineArm_1, loc=<7.5753 0 0>, rot=<-0.0175 -0.0705 -0.0133 -0.9973>}
 dyn_L_splineArm_3 {parent=dyn_L_splineArm_2, loc=<6.895 0 0>, rot=<-0.0627 -0.0093 -0.0562 -0.9964>}
 dyn_L_splineArm_4 {parent=dyn_L_splineArm_3, loc=<6.4776 0 0>, rot=<-0.0769 -0.0347 -0.0753 -0.9936>}
 dyn_L_splineArm_5 {parent=dyn_L_splineArm_4, loc=<5.8029 0 0>, rot=<-0.0221 -0.0259 -0.0245 -0.9991>}
 dyn_L_splineArm_6 {parent=dyn_L_splineArm_5, loc=<5.5535 0 0>, rot=<-0.0193 -0.0531 -0.0251 -0.9981>}
 dyn_L_splineArm_7 {parent=dyn_L_splineArm_6, loc=<5.7419 0 0>, rot=<0.0185 0.0618 0.0237 -0.9976>}
 dyn_L_splineArm_hand {parent=dyn_L_splineArm_7, loc=<5.6578 0 0>, rot=<-0.6835 -0.0107 0.0653 -0.727>}
 dyn_L_splineArm_metacarpal {parent=dyn_L_splineArm_hand, loc=<8.7497 0 0>, rot=<0.0273 0.0616 0.1998 -0.9775>}
 dyn_L_splineArm_finger {parent=dyn_L_splineArm_metacarpal, loc=<3.6352 0 0>, rot=<0.1322 0.0279 0.2044 -0.9695>}
 dyn_L_custom_thumb_01 {parent=dyn_L_splineArm_hand, loc=<0.4624 -1.2238 -2.6072>, rot=<0.7491 -0.4243 0.1493 -0.4862>}
 dyn_L_custom_thumb_02 {parent=dyn_L_custom_thumb_01, loc=<5.9336 -0.0448 0.0401>, rot=<-0.0037 0.0213 0.298 -0.9543>}
 dyn_L_custom_thumb_03 {parent=dyn_L_custom_thumb_02, loc=<2.2223 0.3682 -0.0907>, rot=<0.0118 -0.013 0.0066 -0.9998>}
 dyn_R_splineArm_clavicle {parent=spine_05, loc=<1.3178 -0.8054 10.5135>, rot=<0.7289 0.0465 -0.6829 0.009>}
 dyn_R_splineArm_1 {parent=dyn_R_splineArm_clavicle, loc=<-10.9662 -0.0022 1.3173>, rot=<0.0151 -0.2535 -0.0215 -0.967>}
 dyn_R_splineArm_2 {parent=dyn_R_splineArm_1, loc=<-7.5751 0 -0.0001>, rot=<-0.0175 -0.0705 -0.0133 -0.9973>}
 dyn_R_splineArm_3 {parent=dyn_R_splineArm_2, loc=<-6.8948 0 -0.0002>, rot=<-0.0627 -0.0093 -0.0562 -0.9964>}
 dyn_R_splineArm_4 {parent=dyn_R_splineArm_3, loc=<-6.4781 0.0001 0.0005>, rot=<-0.0769 -0.0347 -0.0753 -0.9936>}
 dyn_R_splineArm_5 {parent=dyn_R_splineArm_4, loc=<-5.8023 -0.0003 -0.0005>, rot=<-0.0221 -0.0259 -0.0245 -0.9991>}
 dyn_R_splineArm_6 {parent=dyn_R_splineArm_5, loc=<-5.5538 0.0002 0.0003>, rot=<-0.0193 -0.0531 -0.0251 -0.9981>}
 dyn_R_splineArm_7 {parent=dyn_R_splineArm_6, loc=<-5.742 0.0001 0.0001>, rot=<0.0185 0.0618 0.0237 -0.9976>}
 dyn_R_splineArm_hand {parent=dyn_R_splineArm_7, loc=<-5.658 0.0001 0.0002>, rot=<-0.6835 -0.0107 0.0653 -0.727>}
 dyn_R_splineArm_metacarpal {parent=dyn_R_splineArm_hand, loc=<-8.7494 -0.0003 0.0002>, rot=<0.0273 0.0616 0.1998 -0.9775>}
 dyn_R_splineArm_finger {parent=dyn_R_splineArm_metacarpal, loc=<-3.6352 0 0>, rot=<0.1322 0.0279 0.2044 -0.9695>}
 dyn_R_custom_thumb_01 {parent=dyn_R_splineArm_hand, loc=<-0.462 1.2235 2.6074>, rot=<0.7491 -0.4243 0.1493 -0.4862>}
 dyn_R_custom_thumb_02 {parent=dyn_R_custom_thumb_01, loc=<-5.9337 0.045 -0.0402>, rot=<-0.0037 0.0213 0.298 -0.9543>}
 dyn_R_custom_thumb_03 {parent=dyn_R_custom_thumb_02, loc=<-2.2222 -0.3682 0.0907>, rot=<0.0118 -0.013 0.0066 -0.9998>}
 attach_cape {parent=spine_05, loc=<9.017 -11.714 0>, rot=<0.0616 -0.7044 -0.0616 -0.7044>}
 attach_backpack {parent=spine_05, loc=<-0.123 -13.473 0>, rot=<-0.0093 -0.707 0.0093 -0.707>}
 dyn_L_custom_wobble_arm {parent=spine_01, loc=<5.8026 -0.4354 -13.2399>, rot=<-0.1958 -0.959 -0.0474 -0.1992>}
 dyn_R_custom_wobble_arm {parent=spine_01, loc=<5.803 -0.4354 13.2399>, rot=<0.959 -0.1958 -0.1992 0.0474>}
 thigh_l {parent=pelvis, loc=<-2.5693 -0.0109 -7.7501>, rot=<0.0012 0.0572 0.0155 -0.9982>}
 calf_l {parent=thigh_l, loc=<-39.0484 0 0>, rot=<0.0006 -0.0161 -0.0391 -0.9991>}
 calf_twist_01_l {parent=calf_l, loc=<-35.813 -0.0967 0.0397>, rot=<0.0237 0.0095 -0.0035 -0.9997>}
 calf_twist_02_l {parent=calf_l, loc=<-18.8064 -0.0967 0.0397>, rot=<0.0237 0.0095 -0.0035 -0.9997>}
 foot_l {parent=calf_l, loc=<-42.561 0.1077 0.4292>, rot=<0.0418 -0.0455 0.0343 -0.9975>}
 ball_l {parent=foot_l, loc=<-6.1599 13.8264 0.6806>, rot=<0.0001 -0.0001 -0.7063 -0.7079>}
 thigh_twist_01_l {parent=thigh_l, loc=<-21.6303 0 0>, rot=<0.0013 0 0 -1>}
 thigh_r {parent=pelvis, loc=<-2.5691 -0.0109 7.7501>, rot=<0.0572 -0.0012 0.9982 0.0155>}
 calf_r {parent=thigh_r, loc=<39.0487 0 0>, rot=<0.0006 -0.0161 -0.0391 -0.9991>}
 calf_twist_01_r {parent=calf_r, loc=<35.8132 0.0967 -0.0397>, rot=<0.0237 0.0095 -0.0035 -0.9997>}
 calf_twist_02_r {parent=calf_r, loc=<18.8064 0.0967 -0.0397>, rot=<0.0237 0.0095 -0.0035 -0.9997>}
 foot_r {parent=calf_r, loc=<42.5612 -0.1077 -0.4292>, rot=<0.0418 -0.0455 0.0343 -0.9975>}
 ball_r {parent=foot_r, loc=<6.1598 -13.8264 -0.6806>, rot=<0.0001 -0.0001 -0.7063 -0.7079>}
 thigh_twist_01_r {parent=thigh_r, loc=<21.6303 0 0>, rot=<0.0013 0 0 -1>}
 ik_foot_root {parent=root, loc=<0 0 0>, rot=<0 0 0 1>}
 ik_foot_l {parent=ik_foot_root, loc=<15.2608 -4.4176 9.4433>, rot=<0.0357 0.7045 0.0284 -0.7082>}
 ik_foot_r {parent=ik_foot_root, loc=<-15.2609 -4.4176 9.4431>, rot=<0.7082 0.0284 -0.7045 0.0357>}
 ik_hand_root {parent=root, loc=<0 0 0>, rot=<0 0 0 1>}
 ik_hand_gun {parent=ik_hand_root, loc=<-51.9729 4.0884 102.6935>, rot=<0.6084 0.1793 0.4613 -0.6205>}
 ik_hand_l {parent=ik_hand_gun, loc=<53.0286 82.1782 35.2051>, rot=<-0.5102 -0.7906 -0.3387 0>}
 ik_hand_r {parent=ik_hand_gun, loc=<0 0 0>, rot=<0 0 0 1>}
 dyn_custom_pelvis {parent=root, loc=<0 -4.6605 71.4933>, rot=<0.0127 0.707 -0.0127 -0.707>}
 dyn_L_custom_thigh {parent=dyn_custom_pelvis, loc=<-4.9856 -0.5439 -9.7794>, rot=<0.0023 0.0373 0.0621 -0.9974>}
 dyn_L_custom_calf {parent=dyn_L_custom_thigh, loc=<-24.0903 0 0>, rot=<-0.0057 0.0178 -0.0614 -0.9979>}
 dyn_L_custom_foot {parent=dyn_L_custom_calf, loc=<-33.382 0 0>, rot=<0.0459 -0.0586 0.0097 -0.9972>}
 dyn_L_custom_ball {parent=dyn_L_custom_foot, loc=<-6.1599 13.8264 0.6806>, rot=<0.0001 -0.0001 -0.7063 -0.7079>}
 dyn_R_custom_thigh {parent=dyn_custom_pelvis, loc=<-4.9856 -0.5439 9.7795>, rot=<0.0373 -0.0023 0.9974 0.0621>}
 dyn_R_custom_calf {parent=dyn_R_custom_thigh, loc=<24.0902 0 0>, rot=<-0.0057 0.0178 -0.0614 -0.9979>}
 dyn_R_custom_foot {parent=dyn_R_custom_calf, loc=<33.382 0 -0.0001>, rot=<0.0459 -0.0586 0.0097 -0.9972>}
 dyn_R_custom_ball {parent=dyn_R_custom_foot, loc=<6.1598 -13.8264 -0.6805>, rot=<0.0001 -0.0001 -0.7063 -0.7079>}
 dyn_custom_spine_01 {parent=dyn_custom_pelvis, loc=<6.0182 -3.674 0>, rot=<0 0 0.034 -0.9994>}
 dyn_custom_spine_02 {parent=dyn_custom_spine_01, loc=<12.442 -0.2818 0>, rot=<0 0 -0.0443 -0.999>}
 dyn_custom_spine_03 {parent=dyn_custom_spine_02, loc=<11.6937 0.9047 0>, rot=<0 0 -0.0385 -0.9993>}
 dyn_L_splineArm_clavicle_ctrl null {type=Null, loc=<0 0 0>, rot=<0 0 0 1>}
 dyn_L_splineArm_clavicle_ctrl {controlType=EULER_TRANSFORM, parent=dyn_L_splineArm_clavicle_ctrl null, type=Control, loc=<0 0 0>, rot=<0 0 0 1>}
 dyn_L_splineArm_1_ctrl null {type=Null, loc=<0 0 0>, rot=<0 0 0 1>}
 dyn_L_splineArm_1_ctrl {
  controlType=EULER_TRANSFORM
  parent=dyn_L_splineArm_1_ctrl null
  type=Control
  loc=<-55.7565 -7.935 119.5811>
  rot=<0.0142 0.0339 -0.0202 0.9991>
 }
 dyn_L_splineArm_2_ctrl null {type=Null, loc=<0 0 0>, rot=<0 0 0 1>}
 dyn_L_splineArm_2_ctrl {
  controlType=EULER_TRANSFORM
  parent=dyn_L_splineArm_2_ctrl null
  type=Control
  loc=<2.0756 -0.9317 127.8796>
  rot=<0.0015 0.2859 -0.0125 0.9582>
 }
 dyn_L_splineArm_3_ctrl null {type=Null, loc=<0 0 0>, rot=<0 0 0 1>}
 dyn_L_splineArm_3_ctrl {
  controlType=EULER_TRANSFORM
  parent=dyn_L_splineArm_3_ctrl null
  type=Control
  loc=<23.8556 7.5637 121.9534>
  rot=<0.0533 0.347 -0.0383 0.9356>
 }
 dyn_L_splineArm_4_ctrl null {type=Null, loc=<0 0 0>, rot=<0 0 0 1>}
 dyn_L_splineArm_4_ctrl {
  controlType=EULER_TRANSFORM
  parent=dyn_L_splineArm_4_ctrl null
  type=Control
  loc=<23.385 10.0542 118.8144>
  rot=<0.1444 0.3438 -0.0106 0.9278>
 }
 dyn_L_splineArm_5_ctrl null {type=Null, loc=<0 0 0>, rot=<0 0 0 1>}
 dyn_L_splineArm_5_ctrl {
  controlType=EULER_TRANSFORM
  parent=dyn_L_splineArm_5_ctrl null
  type=Control
  loc=<29.876 1.0512 115.2614>
  rot=<0.2033 0.3793 0.074 0.8996>
 }
 dyn_L_splineArm_6_ctrl null {type=Null, loc=<0 0 0>, rot=<0 0 0 1>}
 dyn_L_splineArm_6_ctrl {
  controlType=EULER_TRANSFORM
  parent=dyn_L_splineArm_6_ctrl null
  type=Control
  loc=<27.8929 2.157 113.3477>
  rot=<0.2249 0.3985 0.1069 0.8827>
 }
 dyn_L_splineArm_7_ctrl null {type=Null, loc=<0 0 0>, rot=<0 0 0 1>}
 dyn_L_splineArm_7_ctrl {
  controlType=EULER_TRANSFORM
  parent=dyn_L_splineArm_7_ctrl null
  type=Control
  loc=<55.4477 0.4874 99.6978>
  rot=<0.2305 0.4549 0.1125 0.8528>
 }
 dyn_L_splineArm_hand_ctrl null {type=Null, loc=<0 0 0>, rot=<0 0 0 1>}
 dyn_L_splineArm_hand_ctrl {
  controlType=EULER_TRANSFORM
  parent=dyn_L_splineArm_hand_ctrl null
  type=Control
  loc=<35.2128 106.5415 -0.6736>
  rot=<0.1738 0.1438 -0.4055 0.8858>
 }
]
variables [
 L arm bind pose {type=FRigPose, default=(Elements=,HierarchyTopologyVersion=-1,PoseHash=-1)}
 L arm controls {type="TArray<FRigElementKey>"}
 L arm custom bones {type="TArray<FRigElementKey>"}
]
